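# Material function for the AxeOfUnreal pickaxe that adds an animated, glowing grid to incoming material attributes through emissive color, with a blend amount for base color.
# Three local-space planar grid projections, with parameterized tiling, speed and offsets, are combined using axis masks from local normals, a cheap triplanar technique.
# The grayscale grid drives blue and yellow-orange color ramps, selected by a step threshold, then scaled into emissive by blend and multiply parameters.
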MaterialFunction /BRCosmetics/Weapons/FORT_Melee/Pickaxe_AxeOfUnreal_41/Materials/MF_Pickaxe_AxeOfUnreal_42_Techart
graph {nodes [
  Constant()
  n1 = GetMaterialAttributes(None=FunctionInput() {inputName=Result, inputType=FunctionInput_MaterialAttributes}) {attributeGetTypes [#1, #2]}
  n2 = NamedRerouteDeclaration(None=@n1.BaseColor) {name=BaseColorIN, nodeColor=<rgba 0.5938 0 1 1>}
  n3 = ScalarParameter() {defaultValue=1, group=01_Season_42, parameterName=BaseColorBlendAmount, sliderMax=1}
  FunctionOutput(None=SetMaterialAttributes!(MaterialAttributes=@n1.MaterialAttributes, EmissiveColor=NamedRerouteUsage(), BaseColor=LinearInterpolate(B=@n2, Alpha=@n3)) {attributeSetTypes [#1, #2]}) {outputName=SetMaterialAttributes}
  MaterialFunctionCall!(Blend, Base) {materialFunction=/Engine/Functions/Engine_MaterialFunctions03/Blends/Blend_Screen}
  n4 = LinearInterpolate(
   A=LinearInterpolate(A=NamedRerouteUsage(), B=NamedRerouteUsage(), Alpha=NamedRerouteUsage())
   B=NamedRerouteUsage()
   Alpha=NamedRerouteUsage()
  )
  n5 = NamedRerouteUsage()
  n6 = LinearInterpolate(
   A=ColorRamp!(None=@n4) {colorCurve="<self>.ColorCurve", desc=Blue Colors}
   B=ColorRamp(None=@n4) {colorCurve="<self>.ColorCurve", desc=Yellow Orange Colors}
   Alpha=Saturate(None=Step(X=@n5) {constY=0.05})
  ) {constB=0}
  n7 = NamedRerouteUsage()
  MaterialFunctionCall!(Base=@n6, Blend=Add(A=ComponentMask(None=@n7) {b=true, g=true, r=true}, B=@n5)) {materialFunction=/Engine/Functions/Engine_MaterialFunctions03/Blends/Blend_ColorBurn}
  n8 = Abs(None=Transform(None=PixelNormalWS()) {transformSourceType=TRANSFORMSOURCE_World, transformType=TRANSFORM_Local})
  n9 = Constant3Vector() {constant=<rgba 0 0 1 0>}
  NamedRerouteDeclaration(None=SmoothStep(Value=DotProduct(A=@n8, B=@n9)) {constMin=0.5}) {name=ZMask, nodeColor=<rgba 0.0148 0 1 1>}
  n10 = LocalPosition()
  n11 = MaterialFunctionCall!(
   Position Plane=Abs(None=ComponentMask(None=@n10.XYZ) {b=true, g=true})
   Offset=VectorParameter() {group=01_Season_42, parameterName=GlowTileOffset}.RGB
   Rotation
   Tiling=ScalarParameter() {defaultValue=0.5, group=01_Season_42, parameterName=GlowTiling}
   Speed=ScalarParameter() {defaultValue=0.1, group=01_Season_42, parameterName=GlowSpeed}
  ) {materialFunction=&/MF_PlanarProjectedGrid}
  NamedRerouteDeclaration(None=@n11.Grid Lines) {name=YZGridlines, nodeColor=<rgba 1 0 0.0017 1>}
  n12 = LocalPosition()
  n13 = MaterialFunctionCall(
   Position Plane=Abs(None=ComponentMask(None=@n12.XYZ) {g=true, r=true})
   Offset=Add(
    A=VectorParameter() {group=01_Season_42, parameterName=GlowTileOffset}.RGB
    B=VectorParameter() {group=01_Season_42, parameterName=GridZSideOffset}.RGB
   )
   Tiling=ScalarParameter() {defaultValue=0.5, group=01_Season_42, parameterName=GlowTiling}
   Speed=ScalarParameter() {defaultValue=0.1, group=01_Season_42, parameterName=GlowSpeed}
  ) {materialFunction=&MF_PlanarProjectedGrid}
  n14 = NamedRerouteUsage()
  NamedRerouteDeclaration(None=Multiply(A=@n13.Grid Lines, B=@n14)) {name=XYGridlines, nodeColor=<rgba 0 0.1189 1 1>}
  NamedRerouteDeclaration(None=@n1.EmissiveColor) {name=EmissiveIN, nodeColor=<rgba 0 0.4375 1 1>}
  n15 = PixelNormalWS()
  n16 = DotProduct(
   A=Abs(None=Transform(None=@n15) {transformSourceType=TRANSFORMSOURCE_World, transformType=TRANSFORM_Local})
   B=Constant3Vector() {constant=<rgba 0 1 0 0>}
  )
  n17 = Add(A=DotProduct(A=LocalPosition().XYZ, B=Constant3Vector() {constant=<rgba 0 0 1 0>})) {constB=-20}
  NamedRerouteDeclaration(None=Multiply(A=SmoothStep(Value=@n16) {constMin=0.55}, B=Saturate(None=@n17))) {name=YMask, nodeColor=<rgba 0.0944 1 0 1>}
  n18 = LinearInterpolate(A=NamedRerouteUsage(), B=NamedRerouteUsage(), Alpha=NamedRerouteUsage()) {constB=0}
  n19 = NamedRerouteUsage()
  n20 = NamedRerouteUsage()
  n21 = ComponentMask(None=Add(A=LinearInterpolate(A=@n18, B=@n19, Alpha=@n20), B=@n6)) {b=true, g=true, r=true}
  n22 = ScalarParameter() {defaultValue=1, group=01_Season_42, parameterName=EmissiceColorBlendAmount}
  NamedRerouteDeclaration(None=Multiply(A=LinearInterpolate(B=@n21, Alpha=@n22), B=ScalarParameter() {defaultValue=1, group=01_Season_42, parameterName=EmissiceGridMultiply})) {name=EmissiveOUT, nodeColor=<rgba 0 0.0391 1 1>}
  n23 = LocalPosition()
  n24 = MaterialFunctionCall(
   Position Plane=Abs(None=ComponentMask(None=@n23.XYZ) {b=true, r=true})
   Offset=Add(
    A=VectorParameter() {group=01_Season_42, parameterName=GlowTileOffset}.RGB
    B=VectorParameter() {group=01_Season_42, parameterName=GridYSideOffset}.RGB
   )
   Tiling=ScalarParameter() {defaultValue=0.5, group=01_Season_42, parameterName=GlowTiling}
   Speed=ScalarParameter() {defaultValue=0.1, group=01_Season_42, parameterName=GlowSpeed}
  ) {materialFunction=&MF_PlanarProjectedGrid}
  n25 = NamedRerouteUsage()
  NamedRerouteDeclaration(None=Multiply(A=@n24.Grayscale Grid, B=@n25)) {name=XZGrayscaleGrid, nodeColor=<rgba 0.0234 1 0 1>}
  NamedRerouteDeclaration(None=@n11.Grayscale Grid) {name=YZGrayscaleGrid, nodeColor=<rgba 1 0 0.0017 1>}
  NamedRerouteDeclaration(None=Multiply(A=@n13.Grayscale Grid, B=@n14)) {name=XYGrayScaleGrid, nodeColor=<rgba 0 0.1189 1 1>}
  NamedRerouteDeclaration(None=Multiply(A=@n24.Grid Lines, B=@n25)) {name=XZGridLine, nodeColor=<rgba 0.0234 1 0 1>}
  n26 = PixelNormalWS()
  n27 = DotProduct(
   A=Abs(None=Transform(None=@n26) {transformSourceType=TRANSFORMSOURCE_World, transformType=TRANSFORM_Local})
   B=Constant3Vector() {constant=<rgba 1 0 0 0>}
  )
  NamedRerouteDeclaration(None=Multiply(A=SmoothStep(Value=@n27) {constMin=0.55})) {name=XMask, nodeColor=<rgba 1 0 0 1>}
  TextureCoordinate()
 ]}
